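# Animation Blueprint for the Ranked Fury pickaxe that animates its chain and procedurally rattles its two pipes.
# Each update reads the wielding player's MeleeAnim curve to set a swing flag, which blends the chain loop to a faster play rate and mostly widens the shake.
# Time-based sines and random ranges compute per-segment offset, rotation and scale variables for a series of additive Transform Bone nodes along both pipes, with scale replaced at the tips.
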AnimBlueprint /BRCosmetics/Weapons/FORT_Melee/Pickaxe_RankedFury/Meshes/RankedFury_Axe_AnimBP
target_skeleton=&/RankedFury_Axe_Skeleton
graphs [
 AnimGraph {nodes [
   n1 = Sequence Player 'RankedFury_Chain_WEP'() {groupName=Chain, method=SyncGroup, playRate=4, sequence=/BRCosmetics/Weapons/FORT_Melee/Pickaxe_RankedFury/Anim/RankedFury_Chain_WEP}
   n2 = Sequence Player 'RankedFury_Chain_WEP'() {groupName=Chain, method=SyncGroup, sequence=&RankedFury_Chain_WEP}
   n3 = Get Melee Anim Playing() -> (Melee Anim Playing:bool)
   n4 = "Transform (Modify) Bone - Bone: pipe_a_root"!(
    ComponentPose=Local To Component(LocalPose=Blend Poses by bool(BlendPose_0=@n1, BlendPose_1=@n2, BlendTime_0=0.2, BlendTime_1=0.2, bActiveValue=@n3) {blendType=Linear})
    Alpha=1
    Translation="0,0,0"
    Rotation="0,0,0"
   ) -> (Pose) {
    rotationMode=BMM_Additive
    rotationSpace=BCS_BoneSpace
    translationMode=BMM_Additive
    translationSpace=BCS_BoneSpace
    alphaBoolBlend {blendOption=Linear}
    boneToModify {boneName=pipe_a_root}
   }
   n5 = "Transform (Modify) Bone - Bone: pipe_b_root"!(ComponentPose=@n4, Alpha=1, Translation="0,0,0", Rotation="0,0,0") -> (Pose) {
    rotationMode=BMM_Additive
    rotationSpace=BCS_BoneSpace
    translationMode=BMM_Additive
    translationSpace=BCS_BoneSpace
    alphaBoolBlend {blendOption=Linear}
    boneToModify {boneName=pipe_b_root}
   }
   n6 = "Transform (Modify) Bone - Bone: pipe_a_01"!(ComponentPose=@n5, Alpha=1, Translation="0,0,0", Rotation="0,0,0") -> (Pose) {
    rotationMode=BMM_Additive
    rotationSpace=BCS_BoneSpace
    translationMode=BMM_Additive
    translationSpace=BCS_BoneSpace
    alphaBoolBlend {blendOption=Linear}
    boneToModify {boneName=pipe_a_01}
   }
   n7 = "Transform (Modify) Bone - Bone: pipe_b_01"!(ComponentPose=@n6, Alpha=1, Translation="0,0,0", Rotation="0,0,0") -> (Pose) {
    rotationMode=BMM_Additive
    rotationSpace=BCS_BoneSpace
    translationMode=BMM_Additive
    translationSpace=BCS_BoneSpace
    alphaBoolBlend {blendOption=Linear}
    boneToModify {boneName=pipe_b_01}
   }
   n8 = "Transform (Modify) Bone - Bone: pipe_a_02"!(ComponentPose=@n7, Alpha=1, Translation="0,0,0", Rotation="0,0,0") -> (Pose) {
    rotationMode=BMM_Additive
    rotationSpace=BCS_BoneSpace
    translationMode=BMM_Additive
    translationSpace=BCS_BoneSpace
    alphaBoolBlend {blendOption=Linear}
    boneToModify {boneName=pipe_a_02}
   }
   n9 = "Transform (Modify) Bone - Bone: pipe_b_02"!(ComponentPose=@n8, Alpha=1, Translation="0,0,0", Rotation="0,0,0") -> (Pose) {
    rotationMode=BMM_Additive
    rotationSpace=BCS_BoneSpace
    translationMode=BMM_Additive
    translationSpace=BCS_BoneSpace
    alphaBoolBlend {blendOption=Linear}
    boneToModify {boneName=pipe_b_02}
   }
   n10 = "Transform (Modify) Bone - Bone: pipe_a_03"!(ComponentPose=@n9, Alpha=1, Translation="0,0,0", Scale="1,1,1") -> (Pose) {
    rotationMode=BMM_Additive
    rotationSpace=BCS_BoneSpace
    scaleMode=BMM_Replace
    translationMode=BMM_Additive
    translationSpace=BCS_BoneSpace
    alphaBoolBlend {blendOption=Linear}
    boneToModify {boneName=pipe_a_03}
   }
   n11 = "Transform (Modify) Bone - Bone: pipe_b_03"!(ComponentPose=@n10, Alpha=1, Translation="0,0,0", Scale="1,1,1") -> (Pose) {
    rotationMode=BMM_Additive
    rotationSpace=BCS_BoneSpace
    scaleMode=BMM_Replace
    translationMode=BMM_Additive
    translationSpace=BCS_BoneSpace
    alphaBoolBlend {blendOption=Linear}
    boneToModify {boneName=pipe_b_03}
   }
   Output Pose(Result=Component To Local(ComponentPose=@n11)) {layerGroup=None}
  ]}
 EventGraph {nodes [
   Event BlueprintUpdateAnimation()
   > n1 = Sequence()
   > n2 = Cast To FortWeaponPickaxeAthena!(execute, Object=GetOwningActor()) -> (then, CastFailed, AsFort Weapon Pickaxe Athena:Fort Weapon Pickaxe Athena Object Reference) {is=title K2Node_DynamicCast}
   n3 = GetOwner(self:Actor Object Reference=@n2)
   > Is Valid(InputObject=@n3)
   n4 = GetComponentByClass!(self:Actor Object Reference=@n3, ComponentClass:Actor Component Class Reference=/Script/Engine.SkeletalMeshComponent) -> (ReturnValue:Skeletal Mesh Component Object Reference)
   > Is Valid(InputObject=@n4)
   > n5 = Cast To FortPlayerAnimInstance(Object=GetAnimInstance(self=@n4))
   > Set Melee Anim Playing(Melee Anim Playing:bool="float >= float"(A=GetCurveValue(self=@n5, CurveName=MeleeAnim), B=1) {is=title K2Node_CallFunction}) -> (Output_Get:bool)
   n6 = GetGameTimeInSeconds()
   n7 = "Sin (Radians)"(A=float * float(A=float * float(A=1, B=35), B=@n6))
   Set PipeA_End_Vector(execute=@n1.then_1, PipeA_End_Vector:Vector=MakeVector(X=float * float(A=@n7, B=0.15), Y=0)) -> (Output_Get:Vector)
   n8 = "Sin (Radians)"(A=float * float(A=float * float(A=1, B=30), B=@n6))
   > Set PipeB_End_Vector(PipeB_End_Vector:Vector=MakeVector(X=float * float(A=@n8, B=0.2), Y=0)) -> (Output_Get:Vector)
   n9 = Get Melee Anim Playing()
   n10 = "Sin (Radians)"(A=float * float(A=float * float(A=1, B=60), B=GetGameTimeInSeconds()))
   n11 = float * float(A=@n10, B=SelectFloat(A=2, B=0.6, bPickA=@n9))
   n12 = float * float(A=@n10, B=SelectFloat(A=1.7, B=0.5, bPickA=@n9))
   > Set PipeA__Root_Rotator(PipeA__Root_Rotator:Rotator=MakeRotator(Pitch=RandomFloatInRange(Min=@n11, Max=3), Yaw=RandomFloatInRange(Min=@n12, Max=3))) -> (Output_Get:Rotator)
   n13 = "Sin (Radians)"(A=float * float(A=float * float(A=1, B=50), B=GetGameTimeInSeconds()))
   n14 = Get Melee Anim Playing()
   n15 = float * float(A=@n13, B=SelectFloat(A=2.2, B=0.7, bPickA=@n14))
   n16 = float * float(A=@n13, B=SelectFloat(A=1.9, B=0.6, bPickA=@n14))
   > Set PipeB_Root_Rotator(PipeB_Root_Rotator:Rotator=MakeRotator(Pitch=RandomFloatInRange(Min=@n15, Max=2), Yaw=RandomFloatInRange(Min=@n16, Max=2))) -> (Output_Get:Rotator)
   n17 = "Sin (Radians)"(A=float * float(A=float * float(A=1, B=30), B=GetGameTimeInSeconds()))
   n18 = Get Melee Anim Playing()
   n19 = float * float(A=@n17, B=SelectFloat(A=1.5, B=1, bPickA=@n18))
   n20 = float * float(A=@n17, B=SelectFloat(A=2, B=0.6, bPickA=@n18))
   > Set Pipe_Middle_Rotator(Pipe_Middle_Rotator:Rotator=MakeRotator(Pitch=RandomFloatInRange(Min=@n19, Max=1.2), Yaw=RandomFloatInRange(Min=@n20, Max=1.3))) -> (Output_Get:Rotator)
   n21 = Get Melee Anim Playing()
   n22 = float * float(A=float * float(A=1, B=20), B=GetGameTimeInSeconds())
   n23 = RandomFloatInRange(Min=float * float(A="Sin (Radians)"(A=@n22), B=SelectFloat(A=1, B=0.5, bPickA=@n21)), Max=1.5)
   > Set Pipe_End_Rotator(Pipe_End_Rotator:Rotator=MakeRotator(Pitch=@n23, Yaw=@n23)) -> (Output_Get:Rotator)
   n24 = "Sin (Radians)"(A=float * float(A=float * float(A=1, B=20), B=GetGameTimeInSeconds()))
   n25 = RandomFloatInRange(
    Min=float + float(A=float / float(A=@n24, B=10) {is=title K2Node_CallFunction}, B=1) {is=title K2Node_CommutativeAssociativeBinaryOperator}
    Max=1
   )
   > Set Pipe_Scale_Vector(Pipe_Scale_Vector:Vector=MakeVector(X=1, Y=@n25, Z=@n25)) -> (Output_Get:Vector)
   float + float(A=0, B=1)
  ]}
]
variables [
 Melee Anim Playing {type=boolean}
 Pipe_End_Rotator {type=Rotator}
 Pipe_Middle_Rotator {type=Rotator}
 Pipe_Scale_Vector {type=Vector}
 PipeA__Root_Rotator {type=Rotator}
 PipeA_End_Vector {type=Vector}
 PipeB_End_Vector {type=Vector}
 PipeB_Root_Rotator {type=Rotator}
]
